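# Material for the inside of the Summer Slurp Monster player skin, enabled for skeletal meshes and Niagara mesh particles with the Default Lit shading model.
# A lerp blends the body diffuse texture toward a SlurpColor parameter, cyan by default, with SlurpDensity setting the tint strength.
# The body normal map is sampled separately, with no further processing.
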Material /Game/Characters/Player/Male/Large/Bodies/M_LRG_Slurp_Monster_Summer/Materials/TechArt/M_LRG_SlurpMonsterSummerInside
graph {
 properties {
  bUsedWithMeshDeformer=true
  bUsedWithMorphTargets=true
  bUsedWithNiagaraMeshParticles=true
  bUsedWithSkeletalMesh=true
  pixelDepthOffsetMode=PDOM_Legacy
  usedShadingModels=Default Lit
 }
 nodes [
  LinearInterpolate(
   A=TextureSample() {texture=/Game/Characters/Player/Male/Large/Bodies/M_LRG_Slurp_Monster_Summer/Textures/T_M_MED_Slurp_Monster_Summer_Body_D}.RGB
   B=VectorParameter() {parameterName=SlurpColor, defaultValue=<rgba 0.0655 0.8857 0.9948 0>}.RGB
   Alpha=ScalarParameter() {defaultValue=0.376, parameterName=SlurpDensity}
  )
  TextureSample() {
   samplerType=SAMPLERTYPE_Normal
   texture=/Game/Characters/Player/Male/Large/Bodies/M_LRG_Slurp_Monster_Summer/Textures/T_M_MED_Slurp_Monster_Summer_Body_N
  }
 ]
}
